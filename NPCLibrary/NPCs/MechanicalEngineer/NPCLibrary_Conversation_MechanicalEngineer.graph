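# Conversation database for the Mechanical Engineer NPC, defining the greeting dialogue shown when a player talks to them.
# One entry point leads to a speech node with a default line and contextual messages gated by controller requirements.
# First-ever meeting and outfit reactions use named priority curve rows, while general lines check the NPC's per-match conversation count.
# The speech branches through link nodes to shared quest, basic services and Season 29 seasonal activation conversations.
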ConversationDatabase /NPCLibrary/NPCs/MechanicalEngineer/NPCLibrary_Conversation_MechanicalEngineer
entry_points [{tag=Conversation.Entry.MechanicalEngineer, destinations [#1]}]
nodes [
 ConversationEntryPointNode_7 {
  id=#1
  class=ConversationEntryPointNode
  outputs [#2]
  properties {entryTag {tagName=Conversation.Entry.MechanicalEngineer}}
 }
 Speech_C_4 {
  id=#2
  class=Speech_C
  outputs [#3, #4, #5]
  properties {
   nodeName=""
   generalConfig {
    defaultMessage="With the right amount of elbow grease, you can fix anything."
    contextualMessages [
     {
      message=Hey.
      contextRequirements [{requirement {class=FortControllerRequirement_IsFirstEverConversationWithNPC}}]
      priority {curve {rowName=Tandem.Speech.Priority.FirstEverInteraction}}
     }
     {
      message=I'm happy he's back. I promise.
      contextRequirements [{requirement {
         class=FortControllerRequirement_HasCID
         properties {allowedCIds [/BRCosmetics/Athena/Items/Cosmetics/Characters/CID_A_034_Athena_Commando_F_CatBurglar]}
        }}]
      priority {curve {rowName=Tandem.Speech.Priority.OutfitReaction}}
     }
     {
      message=Alright. Let's call it water under the bridge.
      contextRequirements [{requirement {
         class=FortControllerRequirement_HasCID
         properties {allowedCIds [&CID_692_Athena_Commando_M_HenchmanTough]}
        }}]
      priority {curve {rowName=Tandem.Speech.Priority.OutfitReaction}}
     }
     {
      message="SHADOW, GHOST, whatever. You're my friends."
      contextRequirements [{requirement {
         class=FortControllerRequirement_HasCID
         properties {allowedCIds [
           &CID_693_Athena_Commando_M_BuffCat
           /BRCosmetics/Athena/Items/Cosmetics/Characters/CID_691_Athena_Commando_F_TNTina
           /BRCosmetics/Athena/Items/Cosmetics/Characters/CID_690_Athena_Commando_F_Photographer
          ]}
        }}]
      priority {curve {rowName=Tandem.Speech.Priority.OutfitReaction}}
     }
     {
      message="Hi, dad. It's... really good to see you."
      contextRequirements [{requirement {class=FortControllerRequirement_HasCID, properties {allowedCIds [/BRCosmetics/Athena/Items/Cosmetics/Characters/Character_PiteousKicks]}}}]
      priority {curve {rowName=Tandem.Speech.Priority.OutfitReaction}}
     }
     {
      message="With the right amount of elbow grease, you can fix anything."
      contextRequirements [{requirement {
         class=FortControllerRequirement_CheckNPCMatchConversationCount
         properties {nPCUniqueIdTag {tagName=AISpawnerData.Type.Tandem.MechanicalEngineer}}
        }}]
      priority {curve {rowName=None}}
     }
     {
      message="Hard work, focus, luck. That's the key to success."
      contextRequirements [{requirement {
         class=FortControllerRequirement_CheckNPCMatchConversationCount
         properties {nPCUniqueIdTag {tagName=AISpawnerData.Type.Tandem.MechanicalEngineer}}
        }}]
      priority {curve {rowName=None}}
     }
     {
      message=I don't want any family drama... but I guess it's found me.
      contextRequirements [{requirement {
         class=FortControllerRequirement_CheckNPCMatchConversationCount
         properties {nPCUniqueIdTag {tagName=AISpawnerData.Type.Tandem.MechanicalEngineer}}
        }}]
      priority {curve {rowName=None}}
     }
     {
      message=I hate to admit it... but I'm really glad my dad is back.
      contextRequirements [{requirement {
         class=FortControllerRequirement_CheckNPCMatchConversationCount
         properties {nPCUniqueIdTag {tagName=AISpawnerData.Type.Tandem.MechanicalEngineer}}
        }}]
      priority {curve {rowName=None}}
     }
     {
      message=So my dad is... back. Hm. Complicated feelings.
      contextRequirements [{requirement {
         class=FortControllerRequirement_CheckNPCMatchConversationCount
         properties {nPCUniqueIdTag {tagName=AISpawnerData.Type.Tandem.MechanicalEngineer}}
        }}]
      priority {curve {rowName=None}}
     }
    ]
   }
  }
 }
 ConversationLinkNode_21 {
  id=#5
  class=ConversationLinkNode
  properties {remoteEntryTag {tagName=Conversation.Entry.Season29.SeasonalActivation.10c}}
 }
 ConversationLinkNode_0 {id=#4, class=ConversationLinkNode, properties {remoteEntryTag {tagName=Conversation.Entry.BasicServices}}}
 ConversationLinkNode_1 {id=#3, class=ConversationLinkNode, properties {remoteEntryTag {tagName=Conversation.Entry.NPCLibrary.Quests}}}
]
